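; Base component for one screenshot pass on a multipass turntable actor, with stub prepare, cleanup and initialize functions for child passes.
; BeginPass prepares and starts capturing if auto-activate is set, otherwise it ends immediately.
; InitializeCameras gathers child cameras, ordering them by numeric first tag ahead of untagged ones.
; Each CaptureNextCamera call requests one screenshot through the turntable's editor interface; when cameras run out, EndPass cleans up and reports completion.
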(event EventBeginPlay
  (|CallParentFunction))

(event EventTick
  (params (DeltaSeconds :type Float%20%28single-precision%29)))

(fn beginPass
  (Variables|Camera|SetcurrentCameraIndex -1)
  (if (Variables|Activation|GetbAutoActivate)
    (CallFunction|PreparePass)
    (Camera|InitializeCameras)
    (Camera|CaptureNextCamera)
    (else
      (CallFunction|EndPass))))

(fn captureNextCamera
  (Math|Integer|IncrementInt (Variables|Camera|GetcurrentCameraIndex)
    (:%20%20
      (if (Utilities|Array|IsValidIndex (Variables|Camera|GetcurrentCameras) (Variables|Camera|GetcurrentCameraIndex))
        (Class|MultiPassTTEditorInterface|TakeScreenshotUsingCameraComponent(Message) (CallFunction|GetOwningTurntable) (Utilities|Array|Get(acopy) (Variables|Camera|GetcurrentCameras) (Variables|Camera|GetcurrentCameraIndex)) "0, 0, 0" 0.0 self)
        (else
          (CallFunction|EndPass))))))

(fn cleanUpPass
  (return false))

(fn endPass
  (CallFunction|CleanUpPass)
  (Class|MultiPassTTEditorInterface|HandlePassComplete(Message) (CallFunction|GetOwningTurntable) self))

(fn getCurrentRenderCamera
  (if (Utilities|Array|IsValidIndex (Variables|Camera|GetcurrentCameras) (Variables|Camera|GetcurrentCameraIndex))
    (return true (Utilities|Array|Get(acopy) (Variables|Camera|GetcurrentCameras) (Variables|Camera|GetcurrentCameraIndex)))
    (else
      (return false 0))))

(fn getOwningTurntable
  (bind _asmultipass_turntable_actor (Utilities|Casting|CastToMultipassTurntableActor (Components|GetOwner)))
  (return _asmultipass_turntable_actor))

(fn getRenderActors
  (bind _actors (Class|MultiPassTTEditorInterface|GetRenderActorList(Message) (CallFunction|GetOwningTurntable)))
  (return _actors))

(fn getRenderActorsBounds
  (params (offset :type Vector) (minRadius :type Float%20%28double-precision%29))
  (bind (_boundscenter _radius _boxextent) (Class|MultiPassTTEditorInterface|GetRenderActorBounds(Message) (CallFunction|GetOwningTurntable) offset minRadius))
  (return _boundscenter _boxextent _radius))

(fn initialize)

(fn initializeCameras
  (if (< (Utilities|Array|Length (Variables|Camera|GetcurrentCameras)) 1)
    (for _array_element (Components|GetChildrenComponents true)
      (bind _ascamera_component (Utilities|Casting|CastToCameraComponent _array_element))
      (if (Utilities|Array|IsValidIndex (Variables|/Script/Engine.ActorComponent|GetComponentTags _ascamera_component))
        (if (> (Utilities|String|StringToInteger (Utilities|String|ToString(Name) (Utilities|Array|Get(acopy) (Variables|/Script/Engine.ActorComponent|GetComponentTags _ascamera_component)))) -1)
          (Utilities|Array|SetArrayElem (Variables|Camera|GetcurrentCameras) (Utilities|String|StringToInteger (Utilities|String|ToString(Name) (Utilities|Array|Get(acopy) (Variables|/Script/Engine.ActorComponent|GetComponentTags _ascamera_component)))) _ascamera_component true) ; Found an order tag, add in order
          (else
            _)))
      (Utilities|Array|Add (LocalVariables|Getl_unsortedCams) _ascamera_component) ; No Order Tag
    )
    (bind (_arrayindex _arrayelement) (Utilities|Array|ReverseforEachLoop (Variables|Camera|GetcurrentCameras))
      (:LoopBody
        (Utilities|IsValid _arrayelement
          (:Is%20Not%20Valid
            (Utilities|Array|RemoveIndex (Variables|Camera|GetcurrentCameras) _arrayindex) ; Remove all empty indices
          )))
      (:Completed
        (Utilities|Array|AppendArray (Variables|Camera|GetcurrentCameras) (LocalVariables|Getl_unsortedCams)) ; Combine ordered and unordered cameras
        (return (Variables|Camera|GetcurrentCameras))))
    (else
      (return (Variables|Camera|GetcurrentCameras)))))

(fn preparePass
  (return false))
