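; Beauty screenshot pass for the multipass turntable tool, capturing placed cameras, automatic orbit views or both.
; Capturing the next camera defers to the parent for placed cameras, with auto views available when none exist or as a follow-on when both modes are enabled.
; Auto views lerp yaw across 360 degrees by camera index over auto count and request screenshots from the owning turntable.
; The pass ends once the index reaches the combined camera count.
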(event Camera|EventcaptureNextCamera
  (if (> (Utilities|Array|Length (Variables|Camera|GetcurrentCameras)) 0)
    (if (Variables|Turntable|GetusePlacedAndAutoCameras)
      (if (Utilities|Array|IsValidIndex (Variables|Camera|GetcurrentCameras) (+ (Variables|Camera|GetcurrentCameraIndex) 1))
        (|CallParentFunction)
        (else
          _))
      (else
        (|CallParentFunction))))
  (Math|Integer|IncrementInt (Variables|Camera|GetcurrentCameraIndex)
    (:%20%20
      (if (< (Variables|Camera|GetcurrentCameraIndex) (+ (Variables|Turntable|GetautoTurntableCameraCount) (Utilities|Array|Length (Variables|Camera|GetcurrentCameras))))
        (bind _returnvalue (CallFunction|GetAutoTurntableRotationfromIndex (Variables|Camera|GetcurrentCameraIndex) (Variables|Turntable|GetautoTurntableCameraCount)))
        (Class|MultiPassTTEditorInterface|TakeScreenshotfromTransform(Message) (CallFunction|GetOwningTurntable) "0, 0, 0" _returnvalue "0, 0, 0" 0.0 self)
        (else
          (CallFunction|EndPass))))))

(fn getAutoTurntableRotationFromIndex
  (params (currentCamerIndex :type Integer) (TotalCameraCount :type Integer))
  (return (Math|Rotator|Lerp(Rotator) "0, 0, 0" "0,360.000000, 0" (/ (Math|Conversions|ToFloat(Integer) currentCamerIndex) (Math|Conversions|ToFloat(Integer) TotalCameraCount)))))
