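; Multipass turntable screenshot pass that captures a scale comparison image of the render actors.
; It sets the turntable camera's projection mode, then chooses a yaw from the longer horizontal bounds axis and derives a sideways offset from it.
; It feeds offset bounds to MultipassTT_HUD to enable scale drawing, has the owning turntable take the screenshot, and ends the pass on a short timer.
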(event EventBeginPlay
  (|CallParentFunction) ; This node is disabled and will not be called.\nDrag off pins to build functionality.
)

(event EventTick
  (params (DeltaSeconds :type Float%20%28single-precision%29))
  (|CallParentFunction DeltaSeconds) ; This node is disabled and will not be called.\nDrag off pins to build functionality.
)

(event EventbeginPass
  (|CallParentFunction))

(event Camera|EventcaptureNextCamera
  (CallFunction|CaptureScaleComparison))

(fn captureScaleComparison
  (Utilities|FlowControl|Sequence
    (:then_0
      (Class|MultiPassTTEditorInterface|SetCameraProjectionMode(Message) (CallFunction|GetOwningTurntable) false)
      (bind (_boundscenter _boxextent) (RenderActors|GetRenderActorsBounds "0, 0, 0"))
      (LocalVariables|Setl_finalCameraRotation (Math|Rotator|MakeRotator 0.0 0.0 (select (> (.x _boxextent) (.y _boxextent)) -9e+01 0.0)))
      (LocalVariables|Setl_finalOffset (* (Math|Vector|GetRightVector (Math|Rotator|MakeRotator 0.0 0.0 (select (> (.x _boxextent) (.y _boxextent)) -9e+01 0.0))) 9e+01)))
    (:then_1
      (bind _boundscenter_2 (RenderActors|GetRenderActorsBounds (LocalVariables|Getl_finalOffset) (LocalVariables|Getl_minRadius)))
      (bind _asmultipass_tt_hud (Utilities|Casting|CastToMultipassTT_HUD (HUD|GetHUD (Game|GetPlayerController 0))))
      (Class|MultipassTTHUD|SetDoDrawScale _asmultipass_tt_hud true _boundscenter_2))
    (:then_2
      (Class|MultiPassTTEditorInterface|TakeScreenshotfromTransform(Message) (CallFunction|GetOwningTurntable) "0, 0, 0" (LocalVariables|Getl_finalCameraRotation) (LocalVariables|Getl_finalOffset) (LocalVariables|Getl_minRadius) 0)
      (Utilities|Time|SetTimerbyEvent (EventDispatchers|CreateEvent :!event endPass) 0.5))))
