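; Multipass turntable screenshot pass that captures UV images of static and skeletal meshes on the render actors.
; It spawns a TT_UVCapture_Actor on prepare and destroys it on cleanup, clearing the HUD text.
; An index-based, recursive walk over actors and their primitive components hands a mesh to the capture actor and shows its name on the HUD.
; Screenshot-completed delegates chain two capture modes, the second for overlapping UVs, then call EndPass.
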(event EventBeginPlay
  (|CallParentFunction) ; This node is disabled and will not be called.\nDrag off pins to build functionality.
)

(event EventTick
  (params (DeltaSeconds :type Float%20%28single-precision%29))
  (|CallParentFunction DeltaSeconds) ; This node is disabled and will not be called.\nDrag off pins to build functionality.
)

(event Camera|EventcaptureNextCamera
  (bind _actors (RenderActors|GetRenderActors))
  (Variables|Default|SetcurrentRenderActorList _actors)
  (Variables|Default|SetcurrentCaptureCompIndex 0)
  (Variables|Default|SetcurrentCaptureActorIndex -1)
  (Default|UnbindallEventsfromOnScreenshotCompleted (Variables|Default|GetcurrentUVCapture))
  (CallFunction|CaptureNextComponentsUVs))

(fn captureMeshUVs
  (params (newMesh :type Object%20Reference))
  (Utilities|FlowControl|Sequence
    (:then_1
      (Class|TTUVCaptureActor|SetMesh (Variables|Default|GetcurrentUVCapture) (Variables|GetnewMesh))
      (CallFunction|DrawCurrentMeshNametoHUD))
    (:then_2
      (Default|UnbindallEventsfromOnScreenshotCompleted (Variables|Default|GetcurrentUVCapture))
      (Default|BindEventtoOnScreenshotCompleted (Variables|Default|GetcurrentUVCapture) (EventDispatchers|CreateEvent :!event captureOverlapUvs))
      (Class|TTUVCaptureActor|TakeScreenshots (Variables|Default|GetcurrentUVCapture) "NewEnumerator0" ""))))

(fn captureNextComponentsUVs
  (bind _l_comp (LocalVariables|Getl_comp))
  (Utilities|FlowControl|Sequence
    (:then_0
      (if (Utilities|Array|IsValidIndex (Variables|Default|GetcurrentComponentList) (Variables|Default|GetcurrentCaptureCompIndex))
        (LocalVariables|Setl_comp (Utilities|Array|Get(acopy) (Variables|Default|GetcurrentComponentList) (Variables|Default|GetcurrentCaptureCompIndex)))
        (Utilities|FlowControl|Sequence
          (:then_0
            (bind _asstatic_mesh_component (Utilities|Casting|CastToStaticMeshComponent _l_comp)
              (:then
                (Utilities|IsValid (Variables|/Script/Engine.StaticMeshComponent|GetStaticMesh _asstatic_mesh_component)
                  (:Is%20Valid
                    (CallFunction|CaptureMeshUVs (Variables|/Script/Engine.StaticMeshComponent|GetStaticMesh _asstatic_mesh_component)))))
              (:CastFailed
                (bind _asskeletal_mesh_component (Utilities|Casting|CastToSkeletalMeshComponent _l_comp))
                (Utilities|IsValid (Variables|/Script/Engine.SkinnedMeshComponent|GetSkeletalMesh _asskeletal_mesh_component)
                  (:Is%20Valid
                    (CallFunction|CaptureMeshUVs (Variables|/Script/Engine.SkinnedMeshComponent|GetSkeletalMesh _asskeletal_mesh_component)))))))
          (:then_1
            (Math|Integer|IncrementInt (Variables|Default|GetcurrentCaptureCompIndex)
              (:%20%20
                (return)))))
        (else
          (Math|Integer|IncrementInt (Variables|Default|GetcurrentCaptureActorIndex)) ; This actor is complete, go to next actor (we start at -1 to handle the first actor)
        )))
    (:then_1
      (if (Utilities|Array|IsValidIndex (Variables|Default|GetcurrentRenderActorList) (Variables|Default|GetcurrentCaptureActorIndex))
        (Variables|Default|SetcurrentComponentList (Actor|GetComponentsByClass (Utilities|Array|Get(acopy) (Variables|Default|GetcurrentRenderActorList) (Variables|Default|GetcurrentCaptureActorIndex)) "/Script/Engine.PrimitiveComponent"))
        (Variables|Default|SetcurrentCaptureCompIndex 0)
        (CallFunction|CaptureNextComponentsUVs)
        (return)
        (else
          (CallFunction|EndPass))))))

(fn captureOverlapUVs
  (bind _currentuvcapture (Variables|Default|GetcurrentUVCapture))
  (Default|UnbindallEventsfromOnScreenshotCompleted _currentuvcapture)
  (Default|BindEventtoOnScreenshotCompleted _currentuvcapture (EventDispatchers|CreateEvent :!event endPass))
  (Class|TTUVCaptureActor|TakeScreenshots (Variables|Default|GetcurrentUVCapture) "NewEnumerator1" ""))

(fn cleanUpPass
  (Utilities|FlowControl|Sequence
    (:then_0
      (Utilities|IsValid (Variables|Default|GetcurrentUVCapture)
        (:Is%20Valid
          (Actor|DestroyActor (Variables|Default|GetcurrentUVCapture)))))
    (:then_1
      (bind _asmultipass_tt_hud (Utilities|Casting|CastToMultipassTT_HUD (HUD|GetHUD (Game|GetPlayerController 0))))
      (Class|MultipassTTHUD|SetScreenText _asmultipass_tt_hud "" "(R=1.000000,G=0.239216,B=0.039216,A=1.000000)" 1.5 0 :!class /Game/FN_PublicTurntable/MultipassTurntable/Game/MultipassTT_HUD.MultipassTT_HUD_C))
    (:then_2
      (return false))))

(fn drawCurrentMeshNameToHUD
  (bind _currentcapturecompindex (Variables|Default|GetcurrentCaptureCompIndex))
  (bind _currentcomponentlist (Variables|Default|GetcurrentComponentList))
  (bind _output (Utilities|Array|Get(acopy) _currentcomponentlist _currentcapturecompindex))
  (bind _asmultipass_tt_hud (Utilities|Casting|CastToMultipassTT_HUD (HUD|GetHUD (Game|GetPlayerController 0))))
  (LocalVariables|Setl_hud _asmultipass_tt_hud)
  (Utilities|FlowControl|Sequence
    (:then_0
      (bind _asskeletal_mesh_component (Utilities|Casting|CastToSkeletalMeshComponent _output)
        (:then
          (LocalVariables|Setl_meshString (Utilities|GetDisplayName (Variables|/Script/Engine.SkinnedMeshComponent|GetSkeletalMesh _asskeletal_mesh_component))))
        (:CastFailed
          (bind _asstatic_mesh_component (Utilities|Casting|CastToStaticMeshComponent _output))
          (LocalVariables|Setl_meshString (Utilities|GetDisplayName (Variables|/Script/Engine.StaticMeshComponent|GetStaticMesh _asstatic_mesh_component))))))
    (:then_1
      (Class|MultipassTTHUD|SetScreenText (LocalVariables|Getl_hud) (Utilities|String|Append "UVs for: " (LocalVariables|Getl_meshString)) "(R=1.000000,G=0.239216,B=0.039216,A=1.000000)" 1.5 0 :!class /Game/FN_PublicTurntable/MultipassTurntable/Game/MultipassTT_HUD.MultipassTT_HUD_C))))

(fn preparePass
  (Utilities|FlowControl|Sequence
    (:then_0
      (bind _returnvalue (Game|SpawnActorfromClass "/Game/FN_PublicTurntable/TT_UVCapture/TT_UVCapture_Actor.TT_UVCapture_Actor_C" (struct :Location "0, 0, 0" :Rotation "0, 0, 0" :Scale "1.000000,1.000000,1.000000") "AlwaysSpawn" "OverrideRootScale" (CallFunction|GetOwningTurntable) 0))
      (Variables|Default|SetcurrentUVCapture _returnvalue))
    (:then_2
      (return false))))
